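# Megascans preset master material for transmissive surfaces, using the Subsurface shading model and virtual textures, with instanced static mesh support.
# Albedo, Metalness, DRT and Normal texture parameters feed the MF_SSSObjAdjustments function, with only DRT's green channel used, as roughness.
# That function holds the shading adjustments, and the material uses Material Attributes.
# All samples except DRT use the shared world-group wrap sampler.
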Material /Bridge/MSPresets/M_MS_Default_Transmission_Material_VT/M_MS_Default_Transmission_Material_VT
graph {
 properties {
  bUseMaterialAttributes=true
  bUsedWithInstancedStaticMeshes=true
  pixelDepthOffsetMode=PDOM_Legacy
  shadingModel=MSM_Subsurface
  usedShadingModels=Subsurface
 }
 nodes [
  n1 = TextureSampleParameter2D() {
   group=07 - Texture Maps
   parameterName=Albedo
   samplerSource=SSM_Wrap_WorldGroupSettings
   samplerType=SAMPLERTYPE_VirtualColor
   sortPriority=0
   texture=/Bridge/MSPresets/MSVTTextures/DefaultDiffuse
  }
  n2 = TextureSampleParameter2D() {
   group=07 - Texture Maps
   parameterName=Metalness
   samplerSource=SSM_Wrap_WorldGroupSettings
   samplerType=SAMPLERTYPE_VirtualLinearColor
   sortPriority=1
   texture=/Bridge/MSPresets/MSVTTextures/BlackPlaceholder
  }
  n3 = TextureSampleParameter2D() {
   desc=DRT
   group=07 - Texture Maps
   parameterName=DRT
   samplerType=SAMPLERTYPE_VirtualLinearColor
   texture=/Bridge/MSPresets/MSVTTextures/WhitePlaceholder
  }
  n4 = TextureSampleParameter2D() {
   group=07 - Texture Maps
   parameterName=Normal
   samplerSource=SSM_Wrap_WorldGroupSettings
   samplerType=SAMPLERTYPE_VirtualNormal
   sortPriority=3
   texture=/Bridge/MSPresets/MSVTTextures/Placeholder_Normal
  }
  MaterialFunctionCall!(Albedo=@n1.RGB, Metallic=@n2.R, Roughness=@n3.G, Tangent Normal=@n4.RGB) {materialFunction=/Bridge/MSPresets/M_MS_Default_Transmission_Material/Functions/MF_SSSObjAdjustments}
 ]
}
